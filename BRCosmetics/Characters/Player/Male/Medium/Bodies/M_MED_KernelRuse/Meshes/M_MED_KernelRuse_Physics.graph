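# Physics asset for the KernelRuse male medium player body cosmetic, covering four bag bones (B and C, left and right) and both thighs.
# Each bag is a box body constrained to a tiny capsule on its matching MSTR_ bone, with narrow swing limits and very tight twist.
# The thighs carry large unconstrained capsules as leg collision shapes beside the bags; the rest of the skeleton has no bodies.
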PhysicsAsset /BRCosmetics/Characters/Player/Male/Medium/Bodies/M_MED_KernelRuse/Meshes/M_MED_KernelRuse_Physics
bodies {
 MSTR_bag_B_l [{shapeType=Capsule, length=0.505, radius=0.505, shapeName=MSTR_bag_B_l_capsule, rotation=<rot 0 -90 0>}]
 MSTR_bag_B_r [{shapeType=Capsule, length=0.505, radius=0.505, shapeName=MSTR_bag_B_r_capsule, rotation=<rot 0 -90 0>}]
 MSTR_bag_C_l [{
   shapeType=Capsule
   length=0.505
   radius=0.505
   shapeName=MSTR_bag_C_l_capsule
   center=<-1.1406 0.0182 -0.1787>
   rotation=<rot 0 3.2443 -89.4806>
  }]
 MSTR_bag_C_r [{shapeType=Capsule, length=0.505, radius=0.505, shapeName=MSTR_bag_C_r_capsule, rotation=<rot 0 -90 0>}]
 bag_B_l [{shapeType=Box, extentX=4.45, extentY=7.47, extentZ=10.9, shapeName=bag_B_l_box, center=<3 -0.15 -4.1>, rotation=<rot 13.545 0 0>}]
 bag_B_r [{shapeType=Box, extentX=1.911, extentY=7.206, extentZ=9.579, shapeName=bag_B_r_box, center=<1.9 0 -4>, rotation=<rot 11.309 0.133 -2.5>}]
 bag_C_l [{
   shapeType=Box
   extentX=7.5518
   extentY=6.5534
   extentZ=14.9906
   shapeName=bag_C_l_box
   center=<3.9125 -0.294 -5.8001>
   rotation=<rot 0.8425 -90 -10.1675>
  }]
 bag_C_r [{shapeType=Box, extentX=3.723, extentY=7.037, extentZ=12.6, shapeName=bag_C_r_box, center=<3.3 0 -5.5>, rotation=<rot 15 0 0.7>}]
 thigh_l [{shapeType=Capsule, length=9.6261, radius=8.893, shapeName=thigh_l_capsule, center=<-5.8501 -1.3055 0.1821>, rotation=<rot 90 180 -176>}]
 thigh_r [{shapeType=Capsule, length=8.7742, radius=8.3595, shapeName=thigh_r_capsule, center=<5.5879 -0.1252 0.3523>, rotation=<rot 90 180 -179.9997>}]
}
constraints [
 {bone1Name=bag_C_r, bone2Name=MSTR_bag_C_r, swing1LimitDegrees=10, swing2LimitDegrees=12, twistLimitDegrees=1}
 {bone1Name=bag_C_l, bone2Name=MSTR_bag_C_l, swing1LimitDegrees=10, swing2LimitDegrees=12, twistLimitDegrees=1}
 {bone1Name=bag_B_r, bone2Name=MSTR_bag_B_r, swing1LimitDegrees=10, swing2LimitDegrees=12, twistLimitDegrees=1}
 {bone1Name=bag_B_l, bone2Name=MSTR_bag_B_l, swing1LimitDegrees=10, swing2LimitDegrees=12, twistLimitDegrees=1}
]
